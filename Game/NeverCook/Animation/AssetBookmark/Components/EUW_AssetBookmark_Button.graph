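# Editor utility widget for one asset bookmark button, showing a thumbnail and name set from its Object variable in PreConstruct.
# On click, it syncs the Content Browser to that asset when Is QuickBrowse is set, otherwise it toggles Selected.
# Toggle Select and Deselect events both call Refresh, which tints the button gold when selected and white otherwise.
# An OpenEditorForAssets node sits unconnected in the graph.
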EditorUtilityWidgetBlueprint /Game/NeverCook/Animation/AssetBookmark/Components/EUW_AssetBookmark_Button
graphs [EventGraph {nodes [
   "On Clicked (Button_0)"() {is=title K2Node_ComponentBoundEvent}
   > n1 = Branch(Condition=Get Is QuickBrowse() -> (Is QuickBrowse:bool))
   n2 = Get Object() -> (Object:Object Reference)
   > SyncBrowserToObjects(AssetPaths=Make Array("[0]":String=Get Object Path String(Object=@n2)) -> (Array:Array of Strings))
   Toggle Select(execute=@n1.else)
   Deselect() {is=title K2Node_CustomEvent}
   > Set Selected(Selected:bool=false) -> (Output_Get:bool)
   > Refresh()
   Toggle Select() {is=title K2Node_CustomEvent}
   > Set Selected(Selected=NOT Boolean(A=Get Selected() -> (Selected:bool)))
   > Refresh()
   Refresh() {is=title K2Node_CustomEvent}
   > SetBackgroundColor!(
    execute
    self:Button Object Reference=Get Button_0() -> (Button_0:Button Object Reference)
    InBackgroundColor:Linear Color Structure=SelectColor(A=(R=1,G=0.7667,B=0,A=1), B=(R=1,G=1,B=1,A=1), bPickA=Get Selected())
   ) -> (then)
   Event PreConstruct()
   n3 = Get Object()
   > SetAssetByObject!(
    execute
    self:Asset Thumbnail Widget Object Reference=Get AssetThumbnailWidget_30() -> (AssetThumbnailWidget_30:Asset Thumbnail Widget Object Reference)
    Object:Object Reference=@n3
   ) -> (then)
   > "SetText (Text)"(
    self=Get TextBlock_Name() -> (TextBlock_Name:Text Object Reference)
    InText="To Text (Object)"!(InObj:Object Reference=@n3) -> (ReturnValue:Text)
   )
   OpenEditorForAssets!(
    execute
    self:Asset Editor Subsystem Object Reference=Get AssetEditorSubsystem()
    Assets:Array of Object References=Make Array("[0]":Object Reference) -> (Array:Array of Object References)
    OpenedMethod:EAssetTypeActivationOpenedMethod Enum=Edit
   ) -> (then, ReturnValue:bool)
  ]}]
variables [
 Is QuickBrowse {type=boolean}
 Object {type=Object, default=/SparksCosmetics/Instrument/Mic/MarchTreat/Sparks_MarchTreat_Mic}
 Selected {type=boolean}
]
widgets {
 info {rootWidgetClass=&Button, widgetCount=7}
 widgets [
  Button_0 {
   widgetClassPath=&Button
   widget=Button_0
   properties {widgetStyle {
     clickedSlateSound {resourceObject=null}
     disabled {
      bIsDynamicallyLoaded=false
      drawAs=RoundedBox
      imageType=NoImage
      mirroring=NoMirror
      resourceName=None
      resourceObject=null
      tiling=NoTile
      imageSize=<0 0>
      margin=<ltrb 0 0 0 0>
      outlineSettings {
       bUseBrushTransparency=false
       roundingType=FixedRadius
       width=1
       color {colorUseRule=UseColor_ColorTable, specifiedColor=<rgba 0 0 0 0>}
       cornerRadii=<4 4 4 4>
      }
      tintColor {colorUseRule=UseColor_Specified, specifiedColor=<rgba 0.0395 0.0395 0.0395 1>}
      uVRegion {bIsValid=false, max=<0 0>, min=<0 0>}
     }
     disabledForeground {colorUseRule=UseColor_ColorTable, specifiedColor=<rgba 0 0 0 0>}
     hovered {
      bIsDynamicallyLoaded=false
      drawAs=RoundedBox
      imageType=NoImage
      mirroring=NoMirror
      resourceName=None
      resourceObject=null
      tiling=NoTile
      imageSize=<32 32>
      margin=<ltrb 0 0 0 0>
      outlineSettings {
       bUseBrushTransparency=false
       roundingType=FixedRadius
       width=1
       color {colorUseRule=UseColor_ColorTable, specifiedColor=<rgba 0 0 0 0>}
       cornerRadii=<4 4 4 4>
      }
      tintColor {colorUseRule=UseColor_Specified, specifiedColor=<rgba 0.0953 0.0953 0.0953 1>}
      uVRegion {bIsValid=false, max=<0 0>, min=<0 0>}
     }
     hoveredForeground {colorUseRule=UseColor_ColorTable, specifiedColor=<rgba 0 0 0 0>}
     hoveredSlateSound {resourceObject=null}
     normal {
      bIsDynamicallyLoaded=false
      drawAs=RoundedBox
      imageType=NoImage
      mirroring=NoMirror
      resourceName=None
      resourceObject=null
      tiling=NoTile
      imageSize=<32 32>
      margin=<ltrb 0 0 0 0>
      outlineSettings {
       bUseBrushTransparency=false
       roundingType=FixedRadius
       width=1
       color {colorUseRule=UseColor_ColorTable, specifiedColor=<rgba 0 0 0 0>}
       cornerRadii=<4 4 4 4>
      }
      tintColor {colorUseRule=UseColor_Specified, specifiedColor=<rgba 0.2961 0.2961 0.2961 1>}
      uVRegion {bIsValid=false, max=<0 0>, min=<0 0>}
     }
     normalForeground {colorUseRule=UseColor_ColorTable, specifiedColor=<rgba 0 0 0 0>}
     normalPadding=<ltrb 3 3 3 3>
     pressed {
      bIsDynamicallyLoaded=false
      drawAs=RoundedBox
      imageType=NoImage
      mirroring=NoMirror
      resourceName=None
      resourceObject=null
      tiling=NoTile
      imageSize=<32 32>
      margin=<ltrb 0 0 0 0>
      outlineSettings {
       bUseBrushTransparency=false
       roundingType=FixedRadius
       width=1
       color {colorUseRule=UseColor_ColorTable, specifiedColor=<rgba 0 0 0 0>}
       cornerRadii=<4 4 4 4>
      }
      tintColor {colorUseRule=UseColor_Specified, specifiedColor=<rgba 0.0284 0.0284 0.0284 1>}
      uVRegion {bIsValid=false, max=<0 0>, min=<0 0>}
     }
     pressedForeground {colorUseRule=UseColor_ColorTable, specifiedColor=<rgba 0 0 0 0>}
     pressedPadding=<ltrb 3 3 3 3>
     pressedSlateSound {resourceObject=null}
    }}
  }
  CanvasPanel_1 {widgetClassPath=&CanvasPanel, parent=Button_0, widget=CanvasPanel_1, slot {class=ButtonSlot}}
  SizeBox_0 {
   widgetClassPath=&SizeBox
   parent=CanvasPanel_1
   widget=SizeBox_0
   properties {bOverride_HeightOverride=true, bOverride_WidthOverride=true, heightOverride=82, widthOverride=73.6119}
   slot {class=CanvasPanelSlot, properties {bAutoSize=true, layoutData=~}}
  }
  VerticalBox_0 {
   widgetClassPath=&VerticalBox
   parent=SizeBox_0
   widget=VerticalBox_0
   properties {bOverride_Cursor=true}
   slot {class=SizeBoxSlot}
  }
  AssetThumbnailWidget_30 {
   widgetClassPath=/Script/UMGEditor.AssetThumbnailWidget
   bIsVariable=true
   parent=VerticalBox_0
   widget=AssetThumbnailWidget_30
   slot {class=VerticalBoxSlot, properties {size {sizeRule=Fill, value=1}}}
  }
  Border_39 {
   widgetClassPath=&Border
   parent=VerticalBox_0
   widget=Border_39
   properties {
    clipping=ClipToBounds
    horizontalAlignment=HAlign_Center
    verticalAlignment=VAlign_Center
    background {
     bIsDynamicallyLoaded=false
     drawAs=NoDrawType
     imageType=NoImage
     mirroring=NoMirror
     resourceName=None
     resourceObject=null
     tiling=NoTile
     imageSize=<32 32>
     margin=<ltrb 0 0 0 0>
     outlineSettings {
      bUseBrushTransparency=false
      roundingType=HalfHeightRadius
      width=0
      color {colorUseRule=UseColor_Specified, specifiedColor=<rgba 0 0 0 0>}
      cornerRadii=<0 0 0 0>
     }
     tintColor {colorUseRule=UseColor_Specified, specifiedColor=<rgba 1 1 1 1>}
     uVRegion {bIsValid=false, max=<0 0>, min=<0 0>}
    }
    padding=<ltrb 0 0 0 0>
   }
   slot {class=VerticalBoxSlot}
  }
  TextBlock_Name {
   widgetClassPath=&TextBlock
   bIsVariable=true
   parent=Border_39
   widget=TextBlock_Name
   properties {
    text=Text Block
    font {
     bForceMonospaced=false
     bMaterialIsStencil=false
     fontMaterial=null
     fontObject=&Roboto
     letterSpacing=0
     monospacedWidth=1
     size=7.5
     skewAmount=0
     typefaceFontName=Bold
     outlineSettings {
      bApplyOutlineToDropShadows=false
      bMiteredCorners=false
      bSeparateFillAlpha=false
      outlineMaterial=null
      outlineSize=0
      outlineColor=<rgba 0 0 0 1>
     }
    }
   }
   slot {class=BorderSlot, properties {horizontalAlignment=HAlign_Center, verticalAlignment=VAlign_Center, padding=<ltrb 0 0 0 0>}}
  }
 ]
}
