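# Editor utility widget that bookmarks assets related to Sparks instrument item definitions, for browsing or editing selected ones.
# The ID list comes from a details view or Content Browser selection.
# Refresh builds a button section per instrument by walking Asset Registry referencers from its meshes, then from character parts and melee weapons, filtered by class.
# For Vocals, it derives the backpack mesh by replacing "MicStand" in the left-hand mesh path.
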EditorUtilityWidgetBlueprint /Game/NeverCook/Animation/AssetBookmark/EUW_InstrumentBookmarks
graphs [
 EventGraph {nodes [
   n1 = "On Check State Changed (EditorUtilityCheckBox_0)"!() -> (OutputDelegate, then, bIsChecked:bool) {is=title K2Node_ComponentBoundEvent}
   > Set QuickBrowse(QuickBrowse:bool=@n1.bIsChecked) -> (Output_Get:bool)
   > n2 = For Each Button!(Exec:Exec) -> (then:Exec, AsEUW Asset Bookmark Button:EUW Asset Bookmark Button Object Reference, Completed:Exec) {is=title K2Node_MacroInstance}
   > Deselect(self:EUW Asset Bookmark Button Object Reference=@n2)
   > Set Is QuickBrowse!(execute, Is QuickBrowse:bool=Get QuickBrowse() -> (QuickBrowse:bool), self:EUW Asset Bookmark Button Object Reference=@n2) -> (then, Output_Get:bool)
   "On Clicked (EditorUtilityButton_68)"() {is=title K2Node_ComponentBoundEvent}
   > Clear(TargetArray=Get Asset Paths() -> (Asset Paths:Array of Strings))
   > n3 = For Each Button()
   > Branch(Condition=Get Selected!(self:EUW Asset Bookmark Button Object Reference=@n3) -> (Selected:bool))
   > Add Unique(
    TargetArray=Get Asset Paths()
    NewItem=Get Object Path String(Object=Get Object!(self:EUW Asset Bookmark Button Object Reference=@n3) -> (Object:Object Reference))
   )
   SyncBrowserToObjects(execute=@n3.Completed, AssetPaths=Get Asset Paths())
   n4 = "On Property Changed (DetailsView_61)"!() -> (OutputDelegate, then, PropertyName:Name=None) {is=title K2Node_ComponentBoundEvent}
   Refresh Widget(execute=[@n4.then, @n7.Completed])
   "On Clicked (EditorUtilityButton)"() {is=title K2Node_ComponentBoundEvent}
   > Clear(TargetArray:Array of Object References=Get Assets() -> (Assets:Array of Object References))
   > Clear(TargetArray:Array of Strings=Get Asset Paths())
   > n5 = For Each Button()
   > Branch(Condition=Get Selected(self=@n5))
   > Add Unique(TargetArray:Array of Object References=Get Assets(), NewItem:"Object Reference (by ref)"=Get Object(self=@n5))
   OpenEditorForAssets!(
    execute=@n5.Completed
    self:Asset Editor Subsystem Object Reference=Get AssetEditorSubsystem()
    Assets:Array of Object References=Get Assets()
    OpenedMethod:EAssetTypeActivationOpenedMethod Enum=Edit
   ) -> (then, ReturnValue:bool)
   "On Clicked (GetAssets)"() {is=title K2Node_ComponentBoundEvent}
   > Clear(TargetArray:Array of Sparks Instrument Item Def Base Object References=Get SparksIDList() -> (SparksIDList:Array of Sparks Instrument Item Def Base Object References))
   > n6 = GetSelectedAssets()
   > n7 = For Each Loop(Array:Array of Object References=@n6) -> (Array Element:Object Reference)
   > n8 = Cast To SparksInstrumentItemDefBase!(execute, Object=@n7) -> (then, CastFailed, AsSparks Instrument Item Def Base:Sparks Instrument Item Def Base Object Reference) {is=title K2Node_DynamicCast}
   > Add(
    TargetArray:Array of Sparks Instrument Item Def Base Object References=Get SparksIDList()
    NewItem:"Sparks Instrument Item Def Base Object Reference (by ref)"=@n8
   ) {is=title K2Node_CallArrayFunction}
   "On Clicked (EditorUtilityButton_73)"() {is=title K2Node_ComponentBoundEvent}
   > Refresh Widget()
   n9 = Refresh Widget() {is=title K2Node_CustomEvent}
   SetObject(execute=[@n9.then, @n35.then], self=Get DetailsView_61() -> (DetailsView_61:Details View Object Reference), NewObject=Self-Reference())
   > ClearChildren(self=Get GroupList() -> (GroupList:Vertical Box Object Reference))
   > n10 = For Each Loop(Array:Array of Sparks Instrument Item Def Base Object References=Get SparksIDList()) -> (Array Element:Sparks Instrument Item Def Base Object Reference)
   > Is Valid(InputObject=@n10)
   > Clear(TargetArray:Array of Object References=Get ObjectRefList() -> (ObjectRefList:Array of Object References))
   > Add Unique(TargetArray=Get ObjectRefList(), NewItem=@n10)
   > n11 = Sequence!(execute) -> (then_0, then_1, then_2:Exec)
   > n12 = Create Widget!(
    execute:Exec
    WorldContextObject:Object Reference
    Class:User Widget Class Reference=/Game/NeverCook/Animation/AssetBookmark/Components/EUW_AssetBookmark_Section.EUW_AssetBookmark_Section_C
    OwningPlayer:Player Controller Object Reference
    IsContentVisible:bool=True
    Label:Text="To Text (Object)"!(InObj:Object Reference=@n10) -> (ReturnValue:Text)
   ) -> (then:Exec, ReturnValue:EUW Asset Bookmark Section Object Reference)
   > AddChild(self=Get GroupList(), Content=@n12)
   n13 = Get WrapBox_SparksID!(self:EUW Asset Bookmark Section Object Reference=@n12) -> (WrapBox_SparksID:Wrap Box Object Reference)
   > ClearChildren(self=@n13)
   n14 = GetInstrumentMeshes!(execute=@n11.then_1, self, SparksID:Sparks Instrument Item Def Base Object Reference=@n10) -> (then, Meshes:Array of Asset Data Structures)
   > n15 = For Each Loop(Array:Array of Asset Data Structures=@n14) -> (Array Element:Asset Data Structure)
   > Add Unique(TargetArray=Get ObjectRefList(), NewItem=GetAsset(InAssetData=@n15))
   > n16 = GetAssetReferencers!(execute, self, AssetData:"Asset Data Structure (by ref)"=@n15) -> (then, OutAssetData:Array of Asset Data Structures)
   > n17 = For Each Loop(Array=@n16)
   n18 = GetAsset(InAssetData=@n17)
   > Add Unique(TargetArray=Get ObjectRefList(), NewItem=@n18)
   > n19 = Sequence!(execute) -> (then_0, then_1)
   > n20 = Cast To CustomCharacterPart!(execute, Object=@n18) -> (then, CastFailed, AsCustom Character Part:Custom Character Part Object Reference) {is=title K2Node_DynamicCast}
   > n21 = GetPathNameForLoadedAsset(LoadedAsset=@n20)
   > n22 = Get Asset By Object Path!(
    execute
    self:Asset Registry Interface=GetAssetRegistry()
    ObjectPath_PathString:String=@n21
    bIncludeOnlyOnDiskAssets:bool=false
    bSkipARFilteredAssets:bool=true
   ) -> (then, ReturnValue:Asset Data Structure) {is=title K2Node_CallFunction}
   > n23 = GetAssetReferencers(AssetData=@n22)
   > GetDisplayAssets!(execute, self, Array:Array of Asset Data Structures=@n23) -> (then)
   n24 = Cast To FortWeaponMeleeItemDefinition!(execute=@n19.then_1, Object=@n18) -> (then, CastFailed, AsFort Weapon Melee Item Definition:Fort Weapon Melee Item Definition Object Reference) {is=title K2Node_DynamicCast}
   > n25 = GetPathNameForLoadedAsset(LoadedAsset=@n24)
   > n26 = Get Asset By Object Path(self=GetAssetRegistry(), ObjectPath_PathString=@n25, bIncludeOnlyOnDiskAssets=false, bSkipARFilteredAssets=true)
   > n27 = GetAssetReferencers(AssetData=@n26)
   > GetDisplayAssets(Array=@n27)
   n28 = For Each Loop(Exec=@n11.then_2, Array:Array of Object References=Get ObjectRefList()) -> (Array Element:Object Reference)
   > n29 = Sequence!(execute) -> (then_0, then_1, then_2:Exec)
   > n30 = Cast To SparksInstrumentItemDefBase(Object=@n28)
   n31 = Create Widget!(
    execute=[@n34.then, @n30.then, @n32.then]
    WorldContextObject
    Class=/Game/NeverCook/Animation/AssetBookmark/Components/EUW_AssetBookmark_Button.EUW_AssetBookmark_Button_C
    OwningPlayer
    Object:Object Reference=@n28
   ) -> (then, ReturnValue:EUW Asset Bookmark Button Object Reference)
   > AddChild(self=@n13, Content=@n31)
   n32 = Cast To FortWeaponItemDefinition!(execute=@n29.then_1, Object=@n28) -> (then, CastFailed, AsFort Weapon Item Definition:Fort Weapon Item Definition Object Reference) {is=title K2Node_DynamicCast}
   n33 = Get Class(Object=@n28) -> (ReturnValue:Object Class Reference)
   n34 = Branch(
    execute=@n29.then_2
    Condition=OR Boolean!(
     A="Equal (Class)"(A=@n33, B=&SkeletalMesh)
     B="Equal (Class)"(A=@n33, B=/Script/FortniteGame.AthenaPickaxeItemDefinition)
     C:bool="Equal (Class)"(A=@n33, B=/Script/FortniteGame.CustomCharacterPart)
     D:bool="Equal (Class)"(A=@n33, B=/Script/FortniteGame.AthenaBackpackItemDefinition)
     E:bool="Equal (Class)"(A=@n33, B=/Script/Engine.AnimBlueprint)
    ) -> (ReturnValue)
   )
   n35 = Event PreConstruct()
  ]}
 ForEachButton {nodes [
   Inputs!() -> (Exec:Exec)
   > n1 = For Each Loop(Array:Array of Widget Object References=GetAllChildren(self=Get GroupList())) -> (Array Element:Widget Object Reference)
   > n2 = Cast To EUW_AssetBookmark_Section!(execute, Object=@n1) -> (then, CastFailed, AsEUW Asset Bookmark Section:EUW Asset Bookmark Section Object Reference) {is=title K2Node_DynamicCast}
   > n3 = For Each Loop(Array=GetAllChildren(self=Get WrapBox_SparksID(self=@n2)))
   > n4 = Cast To EUW_AssetBookmark_Button!(execute, Object=@n3) -> (then, CastFailed, AsEUW Asset Bookmark Button:EUW Asset Bookmark Button Object Reference) {is=title K2Node_DynamicCast}
   > Outputs!(then:Exec, AsEUW Asset Bookmark Button:EUW Asset Bookmark Button Object Reference=@n4, Completed:Exec=@n1.Completed)
  ]}
 GetAssetReferencers {nodes [
   n1 = GetAssetReferencers!() -> (then, AssetData:"Asset Data Structure (by ref)") {is=title K2Node_FunctionEntry}
   > n2 = Get Referencers!(
    execute
    self:Asset Registry Interface=GetAssetRegistry()
    PackageName:Name=Break Asset Data(AssetData=@n1)
    ReferenceOptions:"Asset Registry Dependency Options Structure (by ref)"=Make Asset Registry Dependency Options(bIncludeSoftPackageReferences=True)
   ) -> (then, OutReferencers:Array of Names, ReturnValue:bool) {is=title K2Node_CallFunction}
   > Branch(Condition=Is Valid Index!(TargetArray:Array of Names=@n2, IndexToTest:int) -> (ReturnValue:bool))
   > n3 = GetAssets!(
    execute
    self:Asset Registry Interface=GetAssetRegistry()
    Filter:"ARFilter Structure (by ref)"=MakeARFilter!(
     PackageNames:Array of Names=@n2
     PackagePaths:Array of Names
     SoftObjectPaths:Array of Soft Object Path Structures
     ClassPaths:Array of Top Level Asset Path Structures
     RecursiveClassPathsExclusionSet:Set of Top Level Asset Path Structures
     ClassNames:Array of Names
     RecursiveClassesExclusionSet:Set of Names
     bRecursivePaths:bool=false
     bRecursiveClasses:bool=false
     bIncludeOnlyOnDiskAssets:bool=false
    ) -> (ReturnValue:ARFilter Structure)
    bSkipARFilteredAssets:bool=true
   ) -> (then, OutAssetData:Array of Asset Data Structures, ReturnValue:bool)
   > Return Node!(execute, OutAssetData:Array of Asset Data Structures=@n3)
  ]}
 GetDisplayAssets {nodes [
   n1 = GetDisplayAssets!() -> (then, Array:Array of Asset Data Structures) {is=title K2Node_FunctionEntry}
   > n2 = For Each Loop(Array:Array of Asset Data Structures=@n1) -> (Array Element:Asset Data Structure)
   > Add Unique(TargetArray=Get ObjectRefList(), NewItem=GetAsset(InAssetData=@n2))
  ]}
 GetInstrumentMeshes {nodes [
   n1 = GetInstrumentMeshes!() -> (then, SparksID:Sparks Instrument Item Def Base Object Reference) {is=title K2Node_FunctionEntry}
   n2 = Get LeftHandMesh!(self:Sparks Instrument Item Def Base Object Reference=@n1) -> (LeftHandMesh:Skeletal Mesh Soft Object Reference)
   > n3 = For Each Loop(Array:Array of Skeletal Mesh Soft Object References=Make Array!(
     "[0]":Skeletal Mesh Soft Object Reference=Get Mesh(self:Sparks Instrument Item Def Base Object Reference=@n1) -> (Mesh:Skeletal Mesh Soft Object Reference)
     "[1]":Skeletal Mesh Soft Object Reference=Get AuxiliaryMesh!(self:Sparks Instrument Item Def Base Object Reference=@n1) -> (AuxiliaryMesh:Skeletal Mesh Soft Object Reference)
     "[2]":Skeletal Mesh Soft Object Reference=@n2
    ) -> (Array:Array of Skeletal Mesh Soft Object References)) -> (Array Element:Skeletal Mesh Soft Object Reference)
   n4 = Sequence!(execute=@n3.Completed) -> (then_0, then_1)
   > Branch(Condition="Equal (Enum)"(
     A:ESparksInstrumentType Enum=GetInstrumentType!(self:Sparks Account Item Definition Object Reference=@n1) -> (ReturnValue:ESparksInstrumentType Enum=Guitar)
     B:ESparksInstrumentType Enum=Vocals
    ))
   > n5 = LoadAsset_Blocking(Asset=@n2)
   > n6 = GetPathNameForLoadedAsset(LoadedAsset=@n5)
   n7 = Split(SourceString=@n6, InStr=/, SearchDir=FromEnd)
   > n8 = Get Asset By Object Path(
    self=GetAssetRegistry()
    ObjectPath_PathString=Append!(A=@n7, B=/, C:String=Replace(SourceString=@n7.RightS, From=MicStand, To=MicStand_Backpack)) -> (ReturnValue)
    bIncludeOnlyOnDiskAssets=false
    bSkipARFilteredAssets=true
   )
   > Add Unique(
    TargetArray:Array of Asset Data Structures=Get MeshesList() -> (MeshesList:Array of Asset Data Structures)
    NewItem:"Asset Data Structure (by ref)"=@n8
   )
   Return Node!(execute=@n4.then_1, Meshes:Array of Asset Data Structures=Get MeshesList())
   Branch(execute=@n3.LoopBody, Condition=IsValidSoftObjectReference(SoftObjectReference=@n3))
   > n9 = LoadAsset_Blocking(Asset=@n3)
   > n10 = GetPathNameForLoadedAsset(LoadedAsset=@n9)
   > n11 = Get Asset By Object Path(self=GetAssetRegistry(), ObjectPath_PathString=@n10, bIncludeOnlyOnDiskAssets=false, bSkipARFilteredAssets=true)
   > Add Unique(TargetArray=Get MeshesList(), NewItem=@n11)
  ]}
]
variables [
 Asset Paths {type="array<string>", default []}
 Assets {type="array<Object>", default []}
 In Element Handle {type=ScriptTypedElementHandle, default {}}
 NewVar {type=SparksPickableSongPart, default=null}
 NewVar_0 {type=SkeletalMesh, default=null}
 ObjectRefList {type="array<Object>", default []}
 QuickBrowse {type=boolean}
 SongID {type=SparksSongItemDefinition, default=/SparksSongTemplates/Items/Songs/SID_Placeholder_98}
 SparksIDList {
  type="array<SparksInstrumentItemDefBase>"
  default [
   null
   null
   null
   /SparksCosmetics/Instrument/Guitar/IndoorLace/Sparks_IndoorLace_Guitar
   /SparksCosmetics/Instrument/Mic/IndoorLace/Sparks_IndoorLace_Mic
   /SparksCosmetics/Instrument/Guitar/PearlTaffy/Sparks_PearlTaffy_Guitar
   /SparksCosmetics/Instrument/Mic/PearlTaffy/Sparks_PearlTaffy_Mic
   /SparksCosmetics/Instrument/Keytar/StarVideo/Sparks_StarVideo_Keytar
   /SparksCosmetics/Instrument/Mic/TaperSlink/Sparks_TaperSlink_Mic
   /SparksCosmetics/Instrument/Keytar/TrackEight/Sparks_TrackEight_Keytar
   /SparksCosmetics/Instrument/Guitar/WesternRush/Sparks_WesternRush_Guitar
  ]
 }
 Target {type=SparksJamEmoteItemDefinition, default=null}
]
widgets {
 info {rootWidgetClass=&VerticalBox, widgetCount=15}
 widgets [
  VerticalBox_0 {widgetClassPath=&VerticalBox, widget=VerticalBox_0, properties {bOverride_Cursor=true}}
  HorizontalBox_289 {widgetClassPath=&HorizontalBox, parent=VerticalBox_0, widget=HorizontalBox_289, slot {class=VerticalBoxSlot}}
  EditorUtilityButton_73 {
   widgetClassPath=&EditorUtilityButton
   parent=HorizontalBox_289
   widget=EditorUtilityButton_73
   properties {widgetStyle {
     clickedSlateSound {resourceObject=null}
     disabled {
      bIsDynamicallyLoaded=false
      drawAs=RoundedBox
      imageType=NoImage
      mirroring=NoMirror
      resourceName=None
      resourceObject=null
      tiling=NoTile
      imageSize=<0 0>
      margin=<ltrb 0 0 0 0>
      outlineSettings {
       bUseBrushTransparency=false
       roundingType=FixedRadius
       width=1
       color {colorUseRule=UseColor_ColorTable, specifiedColor=<rgba 0 0 0 0>}
       cornerRadii=<4 4 4 4>
      }
      tintColor {colorUseRule=UseColor_Specified, specifiedColor=<rgba 0.0395 0.0395 0.0395 1>}
      uVRegion {bIsValid=false, max=<0 0>, min=<0 0>}
     }
     disabledForeground {colorUseRule=UseColor_ColorTable, specifiedColor=<rgba 0 0 0 0>}
     hovered {
      bIsDynamicallyLoaded=false
      drawAs=RoundedBox
      imageType=NoImage
      mirroring=NoMirror
      resourceName=None
      resourceObject=null
      tiling=NoTile
      imageSize=<32 32>
      margin=<ltrb 0 0 0 0>
      outlineSettings {
       bUseBrushTransparency=false
       roundingType=FixedRadius
       width=1
       color {colorUseRule=UseColor_ColorTable, specifiedColor=<rgba 0 0 0 0>}
       cornerRadii=<4 4 4 4>
      }
      tintColor {colorUseRule=UseColor_Specified, specifiedColor=<rgba 0.0953 0.0953 0.0953 1>}
      uVRegion {bIsValid=false, max=<0 0>, min=<0 0>}
     }
     hoveredForeground {colorUseRule=UseColor_ColorTable, specifiedColor=<rgba 0 0 0 0>}
     hoveredSlateSound {resourceObject=null}
     normal {
      bIsDynamicallyLoaded=false
      drawAs=RoundedBox
      imageType=NoImage
      mirroring=NoMirror
      resourceName=None
      resourceObject=null
      tiling=NoTile
      imageSize=<32 32>
      margin=<ltrb 0 0 0 0>
      outlineSettings {
       bUseBrushTransparency=false
       roundingType=FixedRadius
       width=1
       color {colorUseRule=UseColor_ColorTable, specifiedColor=<rgba 0 0 0 0>}
       cornerRadii=<4 4 4 4>
      }
      tintColor {colorUseRule=UseColor_Specified, specifiedColor=<rgba 0.0395 0.0395 0.0395 1>}
      uVRegion {bIsValid=false, max=<0 0>, min=<0 0>}
     }
     normalForeground {colorUseRule=UseColor_ColorTable, specifiedColor=<rgba 0 0 0 0>}
     normalPadding=<ltrb 3 3 3 3>
     pressed {
      bIsDynamicallyLoaded=false
      drawAs=RoundedBox
      imageType=NoImage
      mirroring=NoMirror
      resourceName=None
      resourceObject=null
      tiling=NoTile
      imageSize=<32 32>
      margin=<ltrb 0 0 0 0>
      outlineSettings {
       bUseBrushTransparency=false
       roundingType=FixedRadius
       width=1
       color {colorUseRule=UseColor_ColorTable, specifiedColor=<rgba 0 0 0 0>}
       cornerRadii=<4 4 4 4>
      }
      tintColor {colorUseRule=UseColor_Specified, specifiedColor=<rgba 0.0284 0.0284 0.0284 1>}
      uVRegion {bIsValid=false, max=<0 0>, min=<0 0>}
     }
     pressedForeground {colorUseRule=UseColor_ColorTable, specifiedColor=<rgba 0 0 0 0>}
     pressedPadding=<ltrb 3 3 3 3>
     pressedSlateSound {resourceObject=null}
    }}
   slot {class=HorizontalBoxSlot, properties {padding=<ltrb 0 0 5 0>}}
  }
  TextBlock_18 {
   widgetClassPath=&TextBlock
   parent=EditorUtilityButton_73
   widget=TextBlock_18
   properties {
    justification=Center
    text="↻"
    font {
     bForceMonospaced=false
     bMaterialIsStencil=false
     fontMaterial=null
     fontObject=&Roboto
     letterSpacing=0
     monospacedWidth=1
     size=22.5
     skewAmount=0
     typefaceFontName=Bold
     outlineSettings {
      bApplyOutlineToDropShadows=false
      bMiteredCorners=false
      bSeparateFillAlpha=false
      outlineMaterial=null
      outlineSize=1
      outlineColor=<rgba 1 1 1 1>
     }
    }
    renderTransform {angle=72, scale=<1 1>, shear=<0 0>, translation=<0 0>}
    renderTransformPivot=<0.5081 0.4757>
   }
   slot {class=ButtonSlot, properties {padding=<ltrb 0 0 0 -2>}}
  }
  EditorUtilityButton_68 {
   widgetClassPath=&EditorUtilityButton
   parent=HorizontalBox_289
   widget=EditorUtilityButton_68
   slot {class=HorizontalBoxSlot}
  }
  TextBlock_54 {widgetClassPath=&TextBlock, parent=EditorUtilityButton_68, widget=TextBlock_54, properties {text=Browse}, slot {class=ButtonSlot}}
  EditorUtilityButton {widgetClassPath=&EditorUtilityButton, parent=HorizontalBox_289, widget=EditorUtilityButton, slot {class=HorizontalBoxSlot}}
  TextBlock {widgetClassPath=&TextBlock, parent=EditorUtilityButton, widget=TextBlock, properties {text=Edit}, slot {class=ButtonSlot}}
  EditorUtilityCheckBox_0 {
   widgetClassPath=&EditorUtilityCheckBox
   bIsVariable=true
   parent=HorizontalBox_289
   widget=EditorUtilityCheckBox_0
   slot {class=HorizontalBoxSlot, properties {padding=<ltrb 10 0 0 0>}}
  }
  TextBlock_1 {
   widgetClassPath=&TextBlock
   parent=EditorUtilityCheckBox_0
   widget=TextBlock_1
   properties {text=Quick Browse}
   slot {class=PanelSlot}
  }
  DetailsView_61 {
   widgetClassPath=&DetailsView
   parent=HorizontalBox_289
   widget=DetailsView_61
   properties {propertiesToShow [SparksIDList]}
   slot {class=HorizontalBoxSlot, properties {padding=<ltrb 20 0 0 0>}}
  }
  GetAssets {widgetClassPath=&EditorUtilityButton, parent=HorizontalBox_289, widget=GetAssets, slot {class=HorizontalBoxSlot}}
  RefreshText_1 {
   widgetClassPath=/Script/UMG.RichTextBlock
   parent=GetAssets
   widget=RefreshText_1
   properties {text=Get IDs From Browser, textStyleSet=/Game/UI/ItemReceived/AthenaFrontendStyles}
   slot {class=ButtonSlot}
  }
  EditorUtilityScrollBox_0 {
   widgetClassPath=&EditorUtilityScrollBox
   parent=VerticalBox_0
   widget=EditorUtilityScrollBox_0
   slot {class=VerticalBoxSlot, properties {size {sizeRule=Fill, value=1}}}
  }
  GroupList {
   widgetClassPath=&VerticalBox
   bIsVariable=true
   parent=EditorUtilityScrollBox_0
   widget=GroupList
   properties {bOverride_Cursor=true}
   slot {class=ScrollBoxSlot}
  }
 ]
}
